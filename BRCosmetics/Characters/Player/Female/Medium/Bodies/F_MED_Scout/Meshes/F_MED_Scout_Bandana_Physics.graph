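# Physics asset for the bandana on the F_MED_Scout female player cosmetic body, covering its dynamic bandana bones.
# It has two capsule bodies: a small one on Dyn_Master_Bandana and a larger one on Dyn_Bandana, offset along that bone's X axis.
# A single constraint joins the two bones with matching swing and twist limit angles, bounding how far the bandana can bend and rotate relative to its master bone.
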PhysicsAsset /BRCosmetics/Characters/Player/Female/Medium/Bodies/F_MED_Scout/Meshes/F_MED_Scout_Bandana_Physics
bodies {Dyn_Bandana [{shapeType=Capsule, length=2, radius=2, center=<2 0 0>}], Dyn_Master_Bandana [{shapeType=Capsule, length=0.5, radius=0.5}]}
constraints [{bone1Name=Dyn_Bandana, bone2Name=Dyn_Master_Bandana, swing1LimitDegrees=30, swing2LimitDegrees=30, twistLimitDegrees=30}]
