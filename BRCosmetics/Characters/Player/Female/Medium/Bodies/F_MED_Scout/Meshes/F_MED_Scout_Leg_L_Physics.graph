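# Physics asset for the left pant leg mesh of the female medium Scout body, used to simulate loose pant-leg motion.
# Three capsule bodies cover the pelvis, a tiny master pant-leg anchor bone and the pant-leg bone itself.
# The master bone's swing and twist are locked to the pelvis, while the pant leg hangs from it on a constraint with limited swing and very tight twist.
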PhysicsAsset /BRCosmetics/Characters/Player/Female/Medium/Bodies/F_MED_Scout/Meshes/F_MED_Scout_Leg_L_Physics
bodies {
 Dyn_Master_PantLeg_L [{shapeType=Capsule, length=0.505, radius=0.505}]
 Dyn_PantLeg_L [{shapeType=Capsule, length=9, radius=7.5, center=<1 0 0>, rotation=<rot 90 0 0>}]
 pelvis [{shapeType=Capsule, length=15.2079, radius=13.254, shapeName=pelvis_capsule}]
}
constraints [
 {bone1Name=Dyn_PantLeg_L, bone2Name=Dyn_Master_PantLeg_L, swing1LimitDegrees=20, swing2LimitDegrees=20, twistLimitDegrees=1}
 {
  bone1Name=Dyn_Master_PantLeg_L
  bone2Name=pelvis
  swing1LimitDegrees=45
  swing1Motion=Locked
  swing2LimitDegrees=45
  swing2Motion=Locked
  twistLimitDegrees=45
  twistMotion=Locked
 }
]
