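# Physics asset for the F_MED_Scout female body cosmetic, covering a pouch, a right-leg pouch and its strap.
# Pelvis and right thigh capsules anchor master bones, angularly locked at the pelvis and free at the thigh.
# Both pouches have tight angular limits against their masters.
# A five-box strap chain hangs from the leg master, its first two joints limited to one swing axis, later ones looser on all axes.
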PhysicsAsset /BRCosmetics/Characters/Player/Female/Medium/Bodies/F_MED_Scout/Meshes/F_MED_Scout_Physics
bodies {
 Dyn_Leg_Pouch [{shapeType=Capsule, length=3, radius=1, center=<0 0 1>, rotation=<rot 0 0 90>}]
 Dyn_Master_Leg_Pouch [{shapeType=Capsule, length=0.5, radius=0.5}, {shapeType=Box, extentX=2.2, extentY=20, extentZ=20, center=<2.2574 0.0001 3.4482>}]
 Dyn_Master_Pouch [{shapeType=Capsule, length=0.5, radius=0.5}]
 Dyn_Pouch [{shapeType=Capsule, length=3, radius=2, center=<1 0 0>}]
 Dyn_Strap_Leg_Pouch [{shapeType=Box, extentX=6, extentY=4, extentZ=1, center=<3 0 0>}]
 Dyn_Strap_Leg_Pouch1 [{shapeType=Box, extentX=5, extentY=4, extentZ=1, center=<2.5 0 0>}]
 Dyn_Strap_Leg_Pouch2 [{shapeType=Box, extentX=5, extentY=4, extentZ=1, center=<2.5 0 0>}]
 Dyn_Strap_Leg_Pouch3 [{shapeType=Box, extentX=5.6, extentY=4, extentZ=1, center=<2.8 0 0>}]
 Dyn_Strap_Leg_Pouch4 [{shapeType=Box, extentX=4, extentY=4, extentZ=1, center=<2 0 0>}]
 pelvis [{shapeType=Capsule, length=10, radius=10, shapeName=pelvis_capsule, center=<0.2703 -2.298 1.3579>}]
 thigh_r [{shapeType=Capsule, length=25, radius=8, shapeName=thigh_r_capsule, center=<21.0771 -0.0004 0>, rotation=<rot 90 0 0>}]
}
constraints [
 {bone1Name=Dyn_Leg_Pouch, bone2Name=Dyn_Master_Leg_Pouch, swing1LimitDegrees=0.5, swing2LimitDegrees=10, twistLimitDegrees=0.5}
 {bone1Name=Dyn_Pouch, bone2Name=Dyn_Master_Pouch, swing1LimitDegrees=5, swing2LimitDegrees=2, twistLimitDegrees=2}
 {
  bone1Name=Dyn_Strap_Leg_Pouch
  bone2Name=Dyn_Master_Leg_Pouch
  swing1LimitDegrees=1
  swing1Motion=Locked
  swing2LimitDegrees=20
  twistLimitDegrees=1
  twistMotion=Locked
 }
 {
  bone1Name=Dyn_Strap_Leg_Pouch1
  bone2Name=Dyn_Strap_Leg_Pouch
  swing1LimitDegrees=1
  swing1Motion=Locked
  swing2LimitDegrees=20
  twistLimitDegrees=1
  twistMotion=Locked
 }
 {bone1Name=Dyn_Strap_Leg_Pouch2, bone2Name=Dyn_Strap_Leg_Pouch1, swing1LimitDegrees=20, swing2LimitDegrees=30, twistLimitDegrees=10}
 {bone1Name=Dyn_Strap_Leg_Pouch3, bone2Name=Dyn_Strap_Leg_Pouch2, swing1LimitDegrees=40, swing2LimitDegrees=40, twistLimitDegrees=10}
 {bone1Name=Dyn_Strap_Leg_Pouch4, bone2Name=Dyn_Strap_Leg_Pouch3, swing1LimitDegrees=40, swing2LimitDegrees=40, twistLimitDegrees=10}
 {
  bone1Name=Dyn_Master_Leg_Pouch
  bone2Name=thigh_r
  swing1LimitDegrees=45
  swing1Motion=Free
  swing2LimitDegrees=45
  swing2Motion=Free
  twistLimitDegrees=45
  twistMotion=Free
 }
 {
  bone1Name=Dyn_Master_Pouch
  bone2Name=pelvis
  swing1LimitDegrees=45
  swing1Motion=Locked
  swing2LimitDegrees=45
  swing2Motion=Locked
  twistLimitDegrees=45
  twistMotion=Locked
 }
 {
  bone1Name=thigh_r
  bone2Name=pelvis
  swing1LimitDegrees=45
  swing1Motion=Free
  swing2LimitDegrees=45
  swing2Motion=Free
  twistLimitDegrees=45
  twistMotion=Free
 }
]
