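# Animation Blueprint for the female medium Scout player body, adding secondary physics motion to the pose copied from the attached parent mesh.
# Weighted constraints place bandana, pouch and leg-pouch master bones before the pose is cached.
# Four chained base-bone-space RigidBody nodes simulate body, bandana and both legs (legs partially blended), then a constraint rotates the back collar with the head.
# The update event computes Head_Delta from neck-to-head yaw, but the AnimGraph never reads it.
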AnimBlueprint /BRCosmetics/Characters/Player/Female/Medium/Bodies/F_MED_Scout/Meshes/F_MED_Scout_AnimBP
target_skeleton=&/F_MED_Scout_Skeleton
graphs [
 AnimGraph {nodes [
   n1 = RigidBody(ComponentPose=Local To Component(LocalPose=Use cached pose 'preSimCache'())) {
    bClampLinearTranslationLimitToRefPose=true
    bOverrideWorldGravity=true
    overridePhysicsAsset=&/F_MED_Scout_Physics
    simulationSpace=BaseBoneSpace
    worldSpaceMinimumScale=8
    alphaBoolBlend {blendOption=Linear}
    baseBoneRef {boneName=spine_01}
    componentAppliedLinearAccClamp=<3000 3000 3000>
    componentLinearAccScale=<0.4 0.3 0.3>
    componentLinearVelScale=<0.4 0.3 0.3>
    overrideWorldGravity {z=-980}
   }
   n2 = RigidBody(ComponentPose=@n1) {
    bClampLinearTranslationLimitToRefPose=true
    bOverrideWorldGravity=true
    overridePhysicsAsset=&/F_MED_Scout_Bandana_Physics
    simulationSpace=BaseBoneSpace
    worldSpaceMinimumScale=8
    alphaBoolBlend {blendOption=Linear}
    baseBoneRef {boneName=upperarm_l}
    componentAppliedLinearAccClamp=<3000 3000 3000>
    componentLinearAccScale=<0.4 0.3 0.3>
    componentLinearVelScale=<0.4 0.3 0.3>
    overrideWorldGravity {z=-700}
   }
   n3 = RigidBody(ComponentPose=@n2, Alpha=0.6) {
    bClampLinearTranslationLimitToRefPose=true
    bOverrideWorldGravity=true
    overridePhysicsAsset=&/F_MED_Scout_Leg_L_Physics
    simulationSpace=BaseBoneSpace
    alphaBoolBlend {blendOption=Linear}
    baseBoneRef {boneName=thigh_l}
    componentAppliedLinearAccClamp=<3000 3000 3000>
    componentLinearAccScale=<0.4 0.3 0.3>
    componentLinearVelScale=<0.4 0.3 0.3>
    overrideWorldGravity {z=-980}
   }
   n4 = RigidBody(ComponentPose=@n3, Alpha=0.6) {
    bClampLinearTranslationLimitToRefPose=true
    bOverrideWorldGravity=true
    overridePhysicsAsset=&/F_MED_Scout_Leg_R_Physics
    simulationSpace=BaseBoneSpace
    alphaBoolBlend {blendOption=Linear}
    baseBoneRef {boneName=thigh_r}
    componentAppliedLinearAccClamp=<3000 3000 3000>
    componentLinearAccScale=<0.4 0.3 0.3>
    componentLinearVelScale=<0.4 0.3 0.3>
    overrideWorldGravity {z=-980}
   }
   n5 = "Constraint - Dyn_Collar_bk (R)"(ComponentPose=@n4, Alpha=1, ConstraintWeights_0=1) {
    alpha=0
    alphaBoolBlend {blendOption=Linear}
    boneToModify {boneName=Dyn_Collar_bk}
    constraintSetup [{transformType=Rotation, targetBone {boneName=head}}]
    constraintWeights [1]
   }
   Output Pose(Result=Component To Local(ComponentPose=@n5)) {layerGroup=None}
   n6 = "Constraint - Dyn_Master_Bandana (PPP)"!(
    ComponentPose=Local To Component(LocalPose=Copy Pose From Mesh() {bUseAttachedParent=true})
    Alpha=1
    ConstraintWeights_0=0.541
    ConstraintWeights_1:float=0.231
    ConstraintWeights_2:float=0.228
   ) -> (Pose) {
    alphaBoolBlend {blendOption=Linear}
    boneToModify {boneName=Dyn_Master_Bandana}
    constraintSetup [{targetBone {boneName=upperarm_twist_01_l}}, {targetBone {boneName=upperarm_twist_02_l}}, {targetBone {boneName=upperarm_l}}]
    constraintWeights [1,1,1]
   }
   n7 = "Constraint - Dyn_Master_Pouch (PPPP)"!(
    ComponentPose=@n6
    Alpha=1
    ConstraintWeights_0=0.0713
    ConstraintWeights_1:float=0.0692
    ConstraintWeights_2:float=0.0042
    ConstraintWeights_3:float=0.8553
   ) -> (Pose) {
    alphaBoolBlend {blendOption=Linear}
    boneToModify {boneName=Dyn_Master_Pouch}
    constraintSetup [{targetBone {boneName=spine_01}}, {targetBone {boneName=spine_02}}, {targetBone {boneName=thigh_twist_01_l}}, {targetBone {boneName=pelvis}}]
    constraintWeights [1,1,1,1]
   }
   n8 = "Constraint - Dyn_Master_Leg_Pouch (PPP)"!(ComponentPose=@n7, Alpha=1, ConstraintWeights_0=0.3038, ConstraintWeights_1:float=0.5503, ConstraintWeights_2:float=0.1459) -> (Pose) {
    alphaBoolBlend {blendOption=Linear}
    boneToModify {boneName=Dyn_Master_Leg_Pouch}
    constraintSetup [{targetBone {boneName=thigh_r}}, {targetBone {boneName=thigh_twist_01_r}}, {targetBone {boneName=pelvis}}]
    constraintWeights [1,1,1]
   }
   Save cached pose 'preSimCache'(Pose=Component To Local(ComponentPose=@n8)) {cacheName=preSimCache}
  ]}
 EventGraph {nodes [
   Event BlueprintUpdateAnimation()
   n1 = BreakTransform(InTransform=GetDeltaTransformFromRefPose(self=GetOwnerSkeletalMeshComponent(), BoneName=neck_01, BaseName=head))
   > Set Head_Delta(Head_Delta:double=MapRangeClamped(Value=BreakRotator(InRot=@n1.Rotation).Yaw, InRangeA=15, InRangeB=35)) -> (Output_Get:double)
  ]}
]
variables [Head_Delta {type=number}]
